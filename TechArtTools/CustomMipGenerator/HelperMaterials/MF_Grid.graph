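# MF_Grid is a material function that outputs an antialias-style soft line mask from tiled UVs, with inputs for count, scale, thickness and smoothness.
# UVs are scaled about their center, multiplied by the grid amounts and wrapped with Fmod to form cells.
# SmoothStep on the inverted cell coordinate draws lines at cell starts.
# The final Add uses the R component twice, so only X-axis divisions appear; a Subtract node is unused.
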MaterialFunction /TechArtTools/CustomMipGenerator/HelperMaterials/MF_Grid
graph {nodes [
  n1 = OneMinus(None=FunctionInput(Preview=Constant() {r=0.02}) {bUsePreviewValueAsDefault=true, inputName=GridThickness, inputType=FunctionInput_Scalar})
  n2 = Divide(A=FunctionInput(Preview=Constant() {r=0.2}) {bUsePreviewValueAsDefault=true, inputName=GridSmooth, inputType=FunctionInput_Scalar}) {constB=10}
  n3 = TextureCoordinate()
  n4 = AppendVector(
   A=FunctionInput(Preview=Constant() {r=1.1}) {bUsePreviewValueAsDefault=true, inputName=ScaleGridX, inputType=FunctionInput_Scalar}
   B=FunctionInput(Preview=Constant() {r=1.1}) {bUsePreviewValueAsDefault=true, inputName=ScaleGridY, inputType=FunctionInput_Scalar}
  )
  n5 = FunctionInput(Preview=Constant() {r=8}) {bUsePreviewValueAsDefault=true, inputName=GridAmountX, inputType=FunctionInput_Scalar}
  n6 = FunctionInput(Preview=Constant() {r=8}) {bUsePreviewValueAsDefault=true, inputName=GridAmountY, inputType=FunctionInput_Scalar}
  n7 = Fmod(
   A=Multiply(
    A=MaterialFunctionCall(UVs=@n3, Texture Scale=@n4) {materialFunction=&ScaleUVsByCenter}.UVs
    B=AppendVector(A=@n5, B=@n6)
   )
   B=Constant() {r=1}
  )
  n8 = Abs(None=@n7)
  n9 = MaterialFunctionCall(Float2=SmoothStep(Min=Subtract(A=@n1, B=@n2) {constA=0.8}, Max=Add(A=@n1, B=@n2) {constA=0.8}, Value=OneMinus(None=@n8))) {materialFunction=&BreakOutFloat2Components}
  FunctionOutput(None=Clamp(None=Add(A=@n9.R, B=@n9.R)))
  Subtract(A=@n7) {constB=0.5}
 ]}
