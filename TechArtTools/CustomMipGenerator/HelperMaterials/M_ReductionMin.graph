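# Helper material in the custom mip generator tools that builds an animated wipe comparing a point-sampled texture with its reduced version.
# A time-driven cosine threshold on U steers a Lerp between the sample and MF_MaxReduce's "Min" output, with MF_Grid added using half-texture-size cell counts.
# Despite the ReductionMin name, the MF_MinReduce call is unconnected, as are the derivative nodes.
# No material output connection appears, and the texture is an editor icon placeholder.
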Material /TechArtTools/CustomMipGenerator/HelperMaterials/M_ReductionMin
graph {
 expressions {
  MaterialExpressionAdd_0 {
   class=MaterialExpressionAdd
   inputs [
    A {from=MaterialExpressionMaterialFunctionCall_5, output=Result}
    B {from=MaterialExpressionLinearInterpolate_0}
   ]
  }
  MaterialExpressionAdd_1 {
   class=MaterialExpressionAdd
   inputs [
    A {from=MaterialExpressionDivide_0}
    B {from=""}
   ]
  }
  MaterialExpressionComponentMask_0 {class=MaterialExpressionComponentMask, inputs [None {from=MaterialExpressionTextureCoordinate_1}], properties {r=true}}
  MaterialExpressionComponentMask_1 {class=MaterialExpressionComponentMask, inputs [None {from=MaterialExpressionTextureCoordinate_1}], properties {g=true}}
  MaterialExpressionConstantBiasScale_0 {class=MaterialExpressionConstantBiasScale, inputs [None {from=MaterialExpressionCosine_0}]}
  MaterialExpressionConstant_0 {class=MaterialExpressionConstant, properties {r=0.005}}
  MaterialExpressionConstant_1 {class=MaterialExpressionConstant, properties {r=1}}
  MaterialExpressionConstant_2 {class=MaterialExpressionConstant, properties {r=0.5}}
  MaterialExpressionConvert_0 {
   class=MaterialExpressionConvert
   inputs [
    None {from=MaterialExpressionSaturate_0}
    None {from=MaterialExpressionComponentMask_1}
   ]
   properties {
    nodeName=Make Vector2
    convertInputs [
     {
      type=Scalar
      expressionInput {
       expression=&self:MaterialExpressionSaturate_0
       inputName=None
       mask=0
       maskA=0
       maskB=0
       maskG=0
       maskR=0
       outputIndex=0
      }
     }
     {
      type=Scalar
      expressionInput {
       expression=&self:MaterialExpressionComponentMask_1
       inputName=None
       mask=0
       maskA=0
       maskB=0
       maskG=0
       maskR=0
       outputIndex=0
      }
     }
    ]
    convertOutputs [{type=Vector2}]
   }
  }
  MaterialExpressionConvert_1 {
   class=MaterialExpressionConvert
   inputs [None {from=MaterialExpressionDivide_0}]
   properties {
    nodeName=Break Float2
    convertInputs [{
      type=Vector2
      expressionInput {
       expression=&self:MaterialExpressionDivide_0
       inputName=None
       mask=0
       maskA=0
       maskB=0
       maskG=0
       maskR=0
       outputIndex=0
      }
     }]
    convertOutputs [{type=Scalar}, {type=Scalar}]
   }
  }
  MaterialExpressionCosine_0 {class=MaterialExpressionCosine, inputs [None {from=MaterialExpressionMultiply_1, output=""}], properties {}}
  MaterialExpressionDDX_0 {class=MaterialExpressionDDX, inputs [Value {from=MaterialExpressionMultiply_0, output=""}], properties {}}
  MaterialExpressionDDY_0 {class=MaterialExpressionDDY, inputs [Value {from=MaterialExpressionMultiply_0, output=""}], properties {}}
  MaterialExpressionDivide_0 {
   class=MaterialExpressionDivide
   inputs [
    A {from=MaterialExpressionTextureProperty_0}
    B {from=""}
   ]
  }
  MaterialExpressionLinearInterpolate_0 {
   class=MaterialExpressionLinearInterpolate
   inputs [
    A {from=MaterialExpressionTextureSample_0, output=RGB}
    B {from=MaterialExpressionMaterialFunctionCall_6, output=Min}
    Alpha {from=MaterialExpressionConvert_0}
   ]
  }
  MaterialExpressionMaterialFunctionCall_0 {
   class=MaterialExpressionMaterialFunctionCall
   inputs [
    Tex {from=MaterialExpressionTextureObject_0}
    Level {from=""}
   ]
   properties {materialFunction=&/MF_MinReduce}
  }
  MaterialExpressionMaterialFunctionCall_1 {
   class=MaterialExpressionMaterialFunctionCall
   inputs [
    UVs {from=""}
    Texture Size {from=MaterialExpressionTextureProperty_0}
   ]
   properties {materialFunction=/Engine/Functions/Engine_MaterialFunctions01/Texturing/PointSampledUVs}
  }
  MaterialExpressionMaterialFunctionCall_5 {
   class=MaterialExpressionMaterialFunctionCall
   inputs [
    GridThickness {from=MaterialExpressionConstant_0}
    ScaleGridX {from=MaterialExpressionConstant_1}
    ScaleGridY {from=MaterialExpressionConstant_1}
    GridAmountX {from=MaterialExpressionConvert_1}
    GridAmountY {from=MaterialExpressionConvert_1}
    GridSmooth {from=MaterialExpressionConstant_2}
   ]
   properties {materialFunction=&/MF_Grid}
  }
  MaterialExpressionMaterialFunctionCall_6 {
   class=MaterialExpressionMaterialFunctionCall
   inputs [
    Tex {from=MaterialExpressionTextureObject_0}
    Level {from=""}
   ]
   properties {materialFunction=&/MF_MaxReduce}
  }
  MaterialExpressionMultiply_0 {
   class=MaterialExpressionMultiply
   inputs [
    A {from=MaterialExpressionTextureCoordinate_0}
    B {from=""}
   ]
   properties {constB=0}
  }
  MaterialExpressionMultiply_1 {
   class=MaterialExpressionMultiply
   inputs [
    A {from=MaterialExpressionTime_0}
    B {from=""}
   ]
   properties {constB=0.25}
  }
  MaterialExpressionSaturate_0 {class=MaterialExpressionSaturate, inputs [None {from=MaterialExpressionSign_0}]}
  MaterialExpressionSign_0 {class=MaterialExpressionSign, inputs [None {from=MaterialExpressionSubtract_0, output=""}], properties {}}
  MaterialExpressionSubtract_0 {
   class=MaterialExpressionSubtract
   inputs [
    A {from=MaterialExpressionComponentMask_0}
    B {from=MaterialExpressionConstantBiasScale_0}
   ]
  }
  MaterialExpressionTextureCoordinate_0 {class=MaterialExpressionTextureCoordinate}
  MaterialExpressionTextureCoordinate_1 {class=MaterialExpressionTextureCoordinate}
  MaterialExpressionTextureObject_0 {class=MaterialExpressionTextureObject, inputs [], properties {texture=/Engine/EditorResources/UnrealEdIcon_PhysAsset}}
  MaterialExpressionTextureProperty_0 {class=MaterialExpressionTextureProperty, inputs [Tex {from=MaterialExpressionTextureObject_0, output=""}], properties {}}
  MaterialExpressionTextureSample_0 {
   class=MaterialExpressionTextureSample
   inputs [
    UVs {from=MaterialExpressionMaterialFunctionCall_1, output=Result}
    Tex {from=MaterialExpressionTextureObject_0}
    Apply View MipBias {from=""}
   ]
  }
  MaterialExpressionTime_0 {class=MaterialExpressionTime}
 }
 properties {bAutomaticallySetUsageInEditor=false, usedShadingModels=Default Lit}
}
